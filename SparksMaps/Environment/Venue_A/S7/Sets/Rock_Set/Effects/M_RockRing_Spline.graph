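# Translucent, unlit, two-sided material for glowing Rock Set spline rings on Niagara mesh particles, colored by stage colors or a static color.
# Panning caustic Energy noise, Dust noise, a band edge mask and an audio MPC BeatLifetime pulse form the glow and fade masks.
# SplineThicken plus a randomized sine wave drives WPO, with dust tiling compensating for thickness and Niagara RingScale.
# View-angle fade, depth fade and inverted Fresnel soften opacity.
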Material /SparksMaps/Environment/Venue_A/S7/Sets/Rock_Set/Effects/M_RockRing_Spline
graph {
 properties {
  bAutomaticallySetUsageInEditor=false
  bCastRayTracedShadows=false
  bUsedWithNiagaraMeshParticles=true
  blendMode=BLEND_Translucent
  shadingModel=MSM_Unlit
  twoSided=true
  usedShadingModels=Unlit
 }
 nodes [
  n1 = MaterialFunctionCall!(CustomUPosition) {materialFunction=/SparksStageShow/Systems/StageManager/FXManager/MaterialFunctions/MF_MojoStageColors}
  n2 = Add(A=NamedRerouteUsage(), B=NamedRerouteUsage())
  n3 = ScalarParameter() {defaultValue=1}
  n4 = StaticSwitchParameter(
   True=LinearInterpolate(A=@n1.Primary, B=@n1.Primary, Alpha=Power(Base=@n2, Exp=@n3))
   False=VectorParameter() {parameterName=Color, defaultValue=<rgba 0 0.4641 1 1>}.RGB
  ) {defaultValue=true, parameterName=Use Stage Colors}
  n5 = ScalarParameter() {defaultValue=2, parameterName=BrightnessMaster}
  n6 = NamedRerouteUsage()
  n7 = ScalarParameter() {defaultValue=0.1, parameterName=EnergyIntensity}
  n8 = Multiply(A=NamedRerouteUsage(), B=ScalarParameter() {defaultValue=2, parameterName=BeatIntensity}) {constB=2}
  n9 = Add(
   A=Multiply(A=Multiply(A=@n6, B=@n7), B=Add(A=@n8))
   B=Multiply(A=NamedRerouteUsage(), B=ScalarParameter() {defaultValue=5, parameterName=DustIntensity})
  )
  n10 = NamedRerouteUsage()
  MaterialFunctionCall(Emissive=Multiply(A=Multiply(A=@n4, B=@n5), B=Multiply(A=@n9, B=@n10))) {materialFunction=&MF_FX_EmissiveAdjustments}
  n11 = TextureCoordinate()
  n12 = Multiply(A=OneMinus(None=ComponentMask(None=@n11) {g=true}), B=ScalarParameter() {defaultValue=1, group=DO NOT CHANGE, parameterName=A})
  n13 = ScalarParameter() {group=DO NOT CHANGE, parameterName=B}
  NamedRerouteDeclaration(None=AppendVector(A=ComponentMask(None=@n11) {r=true}, B=Add(A=@n12, B=@n13))) {name=SplineUVs, nodeColor=<rgba 0 1 0.4844 1>}
  n14 = NamedRerouteUsage()
  n15 = AppendVector(A=ScalarParameter() {defaultValue=1, parameterName=EnergyTileX}, B=ScalarParameter() {defaultValue=2, parameterName=EnergyTileY})
  n16 = ScalarParameter() {defaultValue=0.1, parameterName=AlphaSpeedX}
  n17 = ScalarParameter() {defaultValue=-0.1, parameterName=AlphaSpeedY}
  n18 = TextureSampleParameter2D(UVs=Panner(Coordinate=Multiply(A=@n14, B=@n15), Speed=AppendVector(A=@n16, B=@n17)) {speedX=-0.05, speedY=-0.1}) {
   parameterName=NoiseTexture
   samplerType=SAMPLERTYPE_LinearGrayscale
   texture=/Game/Global/Textures/Default/Caustics/T_Nosie_Caustics_Grayscale
  }
  n19 = Divide(A=Subtract(A=@n18.R, B=ScalarParameter() {parameterName=EnergyErode}), B=ScalarParameter() {defaultValue=0.5, parameterName=Tex Divide})
  n20 = Multiply(A=@n18.R) {constB=0.6}
  NamedRerouteDeclaration(None=Saturate(None=Add(A=@n20, B=@n19))) {name=Energy, nodeColor=<rgba 0 1 1 1>}
  n21 = Power(
   Base=CollectionParameter() {collection=/FMAudioReactiveData/AudioAnalysis/MPC_Sparks_Audio_Parameters, parameterName=BeatLifetime}
   Exp=ScalarParameter() {defaultValue=3, parameterName=Param_7}
  )
  NamedRerouteDeclaration(None=MaterialFunctionCall(Input=Cosine(None=@n21), Input Low=Constant() {r=-1}, Input High=Constant() {r=1}, Target Low=Constant(), Target High=Constant() {r=1}) {materialFunction=&RemapValueRange}.Result) {name=BeatAnim, nodeColor=<rgba 1 0 0.7969 1>}
  n22 = NamedRerouteUsage()
  n23 = ScalarParameter() {defaultValue=150, parameterName=SplineThickness}
  n24 = AppendVector(
   A=Multiply(
    A=ScalarParameter() {defaultValue=1, parameterName=DustTileX}
    B=Multiply(A=@n23) {constB=0.01, desc=Counteracts UV scaling from SplineThicken}
   )
   B=Multiply(
    A=ScalarParameter() {defaultValue=15, parameterName=DustTileY}
    B=DynamicParameter() {desc=Counteracts UV scaling from Niagara scaling, paramNames [RingScale,Param2,Param3,Param4]}.RingScale
   )
  )
  n25 = ScalarParameter() {defaultValue=-0.05, parameterName=DustSpeedX}
  n26 = ScalarParameter() {defaultValue=0.2, parameterName=DustSpeedY}
  n27 = TextureSample(UVs=Panner(Coordinate=Multiply(A=@n22, B=@n24), Speed=AppendVector(A=@n25, B=@n26)) {speedX=-0.05, speedY=-0.1}) {texture=/Game/Global/Textures/Default/Noise/T_TileNoise_02}
  n28 = OneMinus(None=ScalarParameter() {defaultValue=1, parameterName=DustThickness})
  NamedRerouteDeclaration(None=Saturate(None=Subtract(A=@n27.R, B=@n28))) {name=Dust, nodeColor=<rgba 0.875 0 1 1>}
  n29 = ScalarParameter() {defaultValue=0.5, parameterName=Width}
  n30 = StaticBool() {value=true}
  n31 = NamedRerouteUsage()
  NamedRerouteDeclaration(None=Power(
    Base=MaterialFunctionCall!(Width=@n29, Sharpness, Offset, Direction Switch=@n30, Compare, Input Coordinates=@n31) {materialFunction=/Engine/Functions/Engine_MaterialFunctions03/Procedurals/GeneratedBand}.Result
    Exp=ScalarParameter() {defaultValue=2, parameterName=EdgeMask_Pow}
   )) {name=EdgeMask, nodeColor=<rgba 1 0 0.7266 1>}
  n32 = NamedRerouteUsage()
  n33 = ParticleRandom()
  n34 = Add(A=Add(A=Time(), B=LinearInterpolate(Alpha=@n33) {constB=100}), B=ComponentMask(None=ActorPositionWS()) {r=true})
  n35 = ScalarParameter() {defaultValue=-0.15, parameterName=Sine Speed}
  n36 = Multiply(A=Add(A=ComponentMask(None=@n32) {g=true}, B=Multiply(A=@n34, B=@n35)), B=ScalarParameter() {defaultValue=1, parameterName=Sine Amount})
  n37 = Multiply(
   A=Multiply(A=Sine(None=@n36), B=Constant3Vector() {constant=<rgba 0 1 0 1>})
   B=ScalarParameter() {defaultValue=1, parameterName=WPO Intensity}
  )
  n38 = ComponentMask(None=NamedRerouteUsage()) {g=true}
  n39 = ScalarParameter() {defaultValue=150, parameterName=SplineThickness}
  n40 = MaterialFunctionCall(WidthBase=@n39, WidthTip=@n39, AngleCorrectedNormals=StaticBool() {value=true}) {materialFunction=&SplineThicken}
  Reroute(None=Add(A=@n40.WorldPositionOffset, B=Multiply(A=@n37, B=@n38)))
  n41 = Multiply(A=NamedRerouteUsage(), B=NamedRerouteUsage())
  n42 = MaterialFunctionCall("FadeSensitivity (optional)"=ScalarParameter() {defaultValue=1, parameterName=EdgeFadeSensitivity}) {materialFunction=&FadeBasedOnViewAngle}
  n43 = Fresnel(ExponentIn=ScalarParameter() {defaultValue=0.75, parameterName=FresnelPow}, Normal=@n40.Normal)
  n44 = Multiply(
   A=DepthFade(Opacity=Multiply(A=@n41, B=@n42.Result), FadeDistance=ScalarParameter() {defaultValue=150, parameterName=DepthFadeDistance}) {fadeDistanceDefault=300}
   B=Subtract(A=OneMinus(None=@n43), B=ScalarParameter() {defaultValue=0.5, parameterName=FresnelOffset})
  )
  n45 = ScalarParameter() {defaultValue=1, parameterName=OpacityMaster}
  Reroute(None=Saturate(None=Multiply(A=@n44, B=@n45)))
 ]
}
