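# Niagara effect for the Sparks Venue_A S7 rock set: a user-scaled ring of drifting mesh boxes and sprite dots around a persistent loop mesh.
# Boxes swirl and bob through vortex and sine gravity, pulsing in shape from the audio parameter collection's BeatLifetime.
# Box and dot colors come from NPC_Sparks stage show colors; dots add curl noise and flicker.
# Box and loop emitters each exist in GPU and CPU versions.
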NiagaraSystem /SparksMaps/Environment/Venue_A/S7/Sets/Rock_Set/Effects/NS_Sparks_RockSet_Pixels
dependencies {
 usedDataInterfaces [&NiagaraDataInterfaceCurve, &NiagaraDataInterfaceVectorCurve]
 usedDynamicInputs [&RandomVector]
 usedModules [&SystemState]
}
emitters {
 Boxes {
  simTarget=GPUComputeSim
  emitterUpdateScript [
   &EmitterState(Life Cycle Mode=::NewEnumerator1, Loop Duration=2, UseLoopDelay=false)
   &SpawnRate(Spawn Probability=0.85, SpawnRate=4, Use Spawn Probability=false)
  ]
  particleSpawnScript [
   &InitializeParticle(
    Color=&RandomRangeLinearColor(
     Minimum=&Multiply_LinearColorByFloat(Linear Color=$NPC.NPC_Sparks.StageShowPrimary, Float=10)
     Maximum=&Multiply_LinearColorByFloat(Linear Color=$NPC.NPC_Sparks.StageShowSecondary, Float=8)
    )
    Lifetime Max=5
    Lifetime Min=3
    Lifetime Mode=::NewEnumerator1
    Mesh Scale Mode=::NewEnumerator1
    Mesh Uniform Scale Max=32
    Mesh Uniform Scale Min=12
    Sprite Rotation Mode=::NewEnumerator2
    Sprite Size Mode=::NewEnumerator1
    Uniform Sprite Size Max=500
    Uniform Sprite Size Min=500
    UsePositionOffset=false
    Write Material Random=true
   )
   &ShapeLocation(
    MaskSpawnGroup=false
    Override Seed=false
    Ring Radius=&Multiply_Float(A=220, B=$User.RingScale)
    Scale Mode:ENiagara_ScaleMode=::NewEnumerator1
    Shape Primitive=::NewEnumerator4
   )
   &AddVelocity(
    Cone Angle=60
    Distribution Along Cone Axis=0.25
    Module.UseConeFalloff=true
    Module.UseVelDistribution=true
    Override Seed=false
    Speed Falloff From Cone Axis=0.5
    Velocity Mode=::NewEnumerator2
    Velocity Speed=&RandomRangeFloat(Minimum=100, Maximum=300)
   ) {enabled=false}
  ]
  particleUpdateScript [
   &ParticleState()
   &ScaleMeshSize(Scale Factor=&VectorFromFloat(Value=&FloatFromCurve(
      FloatCurve=DataInterface() {propertyValues {Curve {keys [
          {arriveTangent=6.2606, interpMode=RCIM_Cubic, leaveTangent=6.2606, tangentMode=RCTM_User, time=0, value=0}
          {arriveTangentWeight=0.0667, interpMode=RCIM_Cubic, leaveTangentWeight=0.0667, tangentMode=RCTM_User, time=0.2, value=0.95}
          {arriveTangent=-0.1, interpMode=RCIM_Cubic, leaveTangent=-0.1, time=0.6, value=0.75}
          {arriveTangentWeight=0.1315, interpMode=RCIM_Cubic, leaveTangentWeight=0.263, tangentMode=RCTM_User, time=0.7, value=0.9}
          {arriveTangent=-0.4076, interpMode=RCIM_Cubic, leaveTangent=-0.4076, tangentMode=RCTM_User, time=1, value=0}
         ]}}}
      Scale Curve=$User.BoxesScale
     )))
   &ScaleMeshSize(Scale Factor=&Lerp_Vector(
     A=<0.5 0.5 0.5>
     B=&RandomRangeVector(Minimum=<0.7 0.7 0.7>, Maximum=<0.25 0.25 1.6>)
     Alpha=&FloatFromCurve(
      FloatCurve=DataInterface() {propertyValues {Curve {keys [
          <at 0 1 RCIM_Cubic>
          {arriveTangentWeight=0.0434, interpMode=RCIM_Cubic, leaveTangentWeight=0.0868, tangentMode=RCTM_User, time=0.05, value=1}
          {arriveTangent=-1.0526, interpMode=RCIM_Cubic, leaveTangent=-1.0526, time=0.2, value=0.4}
          <at 1 0 RCIM_Cubic>
         ]}}}
      CurveIndex=$NPC.NPC_Sparks_Audio_Parameters001.BeatLifetime
     )
    )) {module=ScaleMeshSize002}
   &GravityForce(Gravity=&MakeVector(
     X=0
     Y=0
     Z=&Sine(Normalized Angle=$Emitter.Age, Period=2, Scale=&RandomRangeFloat(Minimum=20, Maximum=60), Bias=0)
    ))
   &Drag(Drag=0.25, Use Linear Drag=true, Use Rotational Drag=false)
   &ScaleColor(
    Scale Alpha=&FloatFromCurve(FloatCurve=DataInterface() {propertyValues {Curve {keys [<at -0.0046 1.0113 RCIM_Cubic>,<at 1.0011 0.002 RCIM_Cubic>]}}})
    Scale RGB=&VectorFromCurve(VectorCurve=DataInterface() {propertyValues {
       XCurve {keys [{arriveTangent=-3.0455, interpMode=RCIM_Cubic, time=0, value=1}, {interpMode=RCIM_Cubic, leaveTangent=-0.8583, time=1, value=0}]}
       YCurve {keys [{arriveTangent=-3.0455, interpMode=RCIM_Cubic, time=0, value=1}, {interpMode=RCIM_Cubic, leaveTangent=-0.8583, time=1, value=0}]}
       ZCurve {keys [{arriveTangent=-3.0455, interpMode=RCIM_Cubic, time=0, value=1}, {interpMode=RCIM_Cubic, leaveTangent=-0.8583, time=1, value=0}]}
      }})
    ScaleA=false
    ScaleRGB=true
   )
   &ScaleColor(
    Color Value To Scale=$Particles.Color
    Scale RGB=&Lerp_Vector(
     A=&MakeVectorFromLinearColorRGB(Color=&Multiply_LinearColorByFloat(Linear Color=$NPC.NPC_Sparks.StageShowPrimary, Float=10))
     B=&MakeVectorFromLinearColorRGB(Color=&Multiply_LinearColorByFloat(Linear Color=$NPC.NPC_Sparks.StageShowTertiary, Float=4))
     Alpha=$StackContext.NormalizedAge
    )
    ScaleA=false
    ScaleRGB=true
   ) {module=ScaleColor002}
   &CurlNoiseForce(Bias Noise=false, Noise Frequency=500, Noise Strength=100, Pan Noise=true, Pan Noise Field=<0 0 100>, Randomization Vector:Vector3f=<0 0 0>) {enabled=false}
   &VortexForce(Origin Pull Amount=10, UseOriginOffset=false, Vortex Force Amount=30)
   &SolveForcesAndVelocity(Clamp Velocity=false, Limit Acceleration=false)
  ]
  rendererClasses [&NiagaraMeshRendererProperties]
  renderers [{rendererClass=&NiagaraMeshRendererProperties, rendererIndex=0}]
 }
 Boxes001 {
  emitterUpdateScript [
   &EmitterState(Life Cycle Mode=::NewEnumerator1, Loop Duration=2, UseLoopDelay=false)
   &SpawnRate(Spawn Probability=0.85, SpawnRate=4, Use Spawn Probability=false)
  ]
  particleSpawnScript [
   &InitializeParticle(
    Color=&RandomRangeLinearColor(
     Minimum=&Multiply_LinearColorByFloat(Linear Color=$NPC.NPC_Sparks.StageShowPrimary, Float=10)
     Maximum=&Multiply_LinearColorByFloat(Linear Color=$NPC.NPC_Sparks.StageShowSecondary, Float=8)
    )
    Lifetime Max=5
    Lifetime Min=3
    Lifetime Mode=::NewEnumerator1
    Mesh Scale Mode=::NewEnumerator1
    Mesh Uniform Scale Max=32
    Mesh Uniform Scale Min=12
    Sprite Rotation Mode=::NewEnumerator2
    Sprite Size Mode=::NewEnumerator1
    Uniform Sprite Size Max=500
    Uniform Sprite Size Min=500
    UsePositionOffset=false
    Write Material Random=true
   )
   &ShapeLocation(
    MaskSpawnGroup=false
    Override Seed=false
    Ring Radius=&Multiply_Float(A=220, B=$User.RingScale)
    Scale Mode:ENiagara_ScaleMode=::NewEnumerator1
    Shape Primitive=::NewEnumerator4
   )
   &AddVelocity(
    Cone Angle=60
    Distribution Along Cone Axis=0.25
    Module.UseConeFalloff=true
    Module.UseVelDistribution=true
    Override Seed=false
    Speed Falloff From Cone Axis=0.5
    Velocity Mode=::NewEnumerator2
    Velocity Speed=&RandomRangeFloat(Minimum=100, Maximum=300)
   ) {enabled=false}
  ]
  particleUpdateScript [
   &ParticleState()
   &ScaleMeshSize(Scale Factor=&VectorFromFloat(Value=&FloatFromCurve(
      FloatCurve=DataInterface() {propertyValues {Curve {keys [
          {arriveTangent=6.2606, interpMode=RCIM_Cubic, leaveTangent=6.2606, tangentMode=RCTM_User, time=0, value=0}
          {arriveTangentWeight=0.0667, interpMode=RCIM_Cubic, leaveTangentWeight=0.0667, tangentMode=RCTM_User, time=0.2, value=0.95}
          {arriveTangent=-0.1, interpMode=RCIM_Cubic, leaveTangent=-0.1, time=0.6, value=0.75}
          {arriveTangentWeight=0.1315, interpMode=RCIM_Cubic, leaveTangentWeight=0.263, tangentMode=RCTM_User, time=0.7, value=0.9}
          {arriveTangent=-0.4076, interpMode=RCIM_Cubic, leaveTangent=-0.4076, tangentMode=RCTM_User, time=1, value=0}
         ]}}}
      Scale Curve=$User.BoxesScale
     )))
   &ScaleMeshSize(Scale Factor=&Lerp_Vector(
     A=<0.5 0.5 0.5>
     B=&RandomRangeVector(Minimum=<0.7 0.7 0.7>, Maximum=<0.25 0.25 1.6>)
     Alpha=&FloatFromCurve(
      FloatCurve=DataInterface() {propertyValues {Curve {keys [
          <at 0 1 RCIM_Cubic>
          {arriveTangentWeight=0.0434, interpMode=RCIM_Cubic, leaveTangentWeight=0.0868, tangentMode=RCTM_User, time=0.05, value=1}
          {arriveTangent=-1.0526, interpMode=RCIM_Cubic, leaveTangent=-1.0526, time=0.2, value=0.4}
          <at 1 0 RCIM_Cubic>
         ]}}}
      CurveIndex=$NPC.NPC_Sparks_Audio_Parameters001.BeatLifetime
     )
    )) {module=ScaleMeshSize002}
   &GravityForce(Gravity=&MakeVector(
     X=0
     Y=0
     Z=&Sine(Normalized Angle=$Emitter.Age, Period=2, Scale=&RandomRangeFloat(Minimum=20, Maximum=60), Bias=0)
    ))
   &Drag(Drag=0.25, Use Linear Drag=true, Use Rotational Drag=false)
   &ScaleColor(
    Scale Alpha=&FloatFromCurve(FloatCurve=DataInterface() {propertyValues {Curve {keys [<at -0.0046 1.0113 RCIM_Cubic>,<at 1.0011 0.002 RCIM_Cubic>]}}})
    Scale RGB=&VectorFromCurve(VectorCurve=DataInterface() {propertyValues {
       XCurve {keys [{arriveTangent=-3.0455, interpMode=RCIM_Cubic, time=0, value=1}, {interpMode=RCIM_Cubic, leaveTangent=-0.8583, time=1, value=0}]}
       YCurve {keys [{arriveTangent=-3.0455, interpMode=RCIM_Cubic, time=0, value=1}, {interpMode=RCIM_Cubic, leaveTangent=-0.8583, time=1, value=0}]}
       ZCurve {keys [{arriveTangent=-3.0455, interpMode=RCIM_Cubic, time=0, value=1}, {interpMode=RCIM_Cubic, leaveTangent=-0.8583, time=1, value=0}]}
      }})
    ScaleA=false
    ScaleRGB=true
   )
   &ScaleColor(
    Color Value To Scale=$Particles.Color
    Scale RGB=&Lerp_Vector(
     A=&MakeVectorFromLinearColorRGB(Color=&Multiply_LinearColorByFloat(Linear Color=$NPC.NPC_Sparks.StageShowPrimary, Float=10))
     B=&MakeVectorFromLinearColorRGB(Color=&Multiply_LinearColorByFloat(Linear Color=$NPC.NPC_Sparks.StageShowTertiary, Float=4))
     Alpha=$StackContext.NormalizedAge
    )
    ScaleA=false
    ScaleRGB=true
   ) {module=ScaleColor002}
   &CurlNoiseForce(Bias Noise=false, Noise Frequency=500, Noise Strength=100, Pan Noise=true, Pan Noise Field=<0 0 100>, Randomization Vector:Vector3f=<0 0 0>) {enabled=false}
   &VortexForce(Origin Pull Amount=10, UseOriginOffset=false, Vortex Force Amount=30)
   &SolveForcesAndVelocity(Clamp Velocity=false, Limit Acceleration=false)
  ]
  rendererClasses [&NiagaraMeshRendererProperties]
  renderers [{rendererClass=&NiagaraMeshRendererProperties, rendererIndex=0}]
 }
 DotRing {
  simTarget=GPUComputeSim
  emitterUpdateScript [
   &EmitterState(Life Cycle Mode=::NewEnumerator1, Loop Duration=2, UseLoopDelay=false)
   &SpawnRate(SpawnRate=16, Use Spawn Probability=false)
  ]
  particleSpawnScript [
   &InitializeParticle(
    Color=&RandomRangeLinearColor(
     Minimum=&Multiply_LinearColorByFloat(Linear Color=$NPC.NPC_Sparks.StageShowPrimary, Float=12)
     Maximum=&Multiply_LinearColorByFloat(Linear Color=$NPC.NPC_Sparks.StageShowTertiary, Float=9)
    )
    Lifetime Max=1.75
    Lifetime Min=1.4
    Lifetime Mode=::NewEnumerator1
    Mass Mode=::NewEnumerator3
    Sprite Rotation Mode=::NewEnumerator2
    Sprite Size Mode=::NewEnumerator1
    Uniform Sprite Size Max=4
    Uniform Sprite Size Min=1
    UsePositionOffset=false
    Write Material Random=true
   )
   &ShapeLocation(
    Disc Coverage=0.119
    MaskSpawnGroup=false
    Override Seed=false
    Ring Radius=&Multiply_Float(A=220, B=$User.RingScale)
    Shape Primitive=::NewEnumerator4
    UseOwnerScale=true
   )
  ]
  particleUpdateScript [
   &ParticleState()
   &ScaleSpriteSize(Uniform Curve Sprite Scale=DataInterface() {propertyValues {Curve {keys [
        {interpMode=RCIM_Cubic, tangentMode=RCTM_User, time=0, value=0}
        {interpMode=RCIM_Cubic, tangentMode=RCTM_User, time=0.1, value=1}
        {arriveTangentWeight=0.3986, interpMode=RCIM_Cubic, leaveTangentWeight=0.0498, tangentMode=RCTM_User, time=0.9, value=1}
        {tangentMode=RCTM_User, time=1, value=0}
       ]}}})
   &Drag(Drag=0.25, Use Linear Drag=true, Use Rotational Drag=false)
   &ScaleColor(
    Scale Alpha=&FloatFromCurve(FloatCurve=DataInterface() {propertyValues {Curve {keys [
         <at 0 0.05 RCIM_Cubic>
         {arriveTangentWeight=0.0581, interpMode=RCIM_Cubic, leaveTangentWeight=0.349, tangentMode=RCTM_User, time=0.1, value=1}
         {arriveTangent=-0.375, interpMode=RCIM_Cubic, leaveTangent=-0.375, time=0.2, value=0.675}
         {arriveTangent=-0.4375, interpMode=RCIM_Cubic, leaveTangent=-0.4375, time=0.3, value=0.925}
         {arriveTangent=0.1871, interpMode=RCIM_Cubic, leaveTangent=0.1871, time=0.6, value=0.5}
         {arriveTangentWeight=0.2004, interpMode=RCIM_Cubic, leaveTangentWeight=0.0998, tangentMode=RCTM_User, time=0.7009, value=1}
         {arriveTangent=-4.4292, interpMode=RCIM_Cubic, leaveTangent=-4.4292, time=0.8, value=0.7}
         {arriveTangent=-8.0089, interpMode=RCIM_Cubic, leaveTangent=-8.0089, tangentMode=RCTM_User, time=1, value=-0.325}
        ]}}})
    ScaleA=true
    ScaleRGB=true
   )
   &VortexForce(UseOriginOffset=false, Vortex Force Amount=60)
   &CurlNoiseForce(Bias Noise=false, Noise Strength=60, Pan Noise=false)
   &SolveForcesAndVelocity(Clamp Velocity=false, Limit Acceleration=false)
  ]
  rendererClasses [&NiagaraSpriteRendererProperties]
  renderers [{rendererClass=&NiagaraSpriteRendererProperties}]
 }
 SplineThicken_Loop {
  simTarget=GPUComputeSim
  emitterUpdateScript [
   &EmitterState(Life Cycle Mode=::NewEnumerator1, Loop Behavior=::NewEnumerator1, Loop Duration Mode=::NewEnumerator1, UseLoopDelay=false)
   &SpawnBurst_Instantaneous(Use Spawn Probability=false, UseLoopCountLimit=false)
  ]
  particleSpawnScript [
   &InitializeParticle(
    Color Mode=::NewEnumerator0
    Lifetime=3
    Mesh Scale Mode=::NewEnumerator3
    Mesh Uniform Scale=$User.RingScale
    UsePositionOffset=false
    Write Material Random=true
   )
   &DynamicMaterialParameters(
    Index 0 Param 1=$User.RingScale
    Module.Index0Param0Write=true
    Module.Index0Param1Write=true
    Module.Index0Param2Write=true
    Module.Index0Param3Write=true
   )
   &InitialMeshOrientation(Use Rotation Vector=false) {enabled=false}
  ]
  particleUpdateScript [&ParticleState(Kill Particles When Lifetime Has Elapsed=false, Let Infinitely Lived Particles Die When Emitter Deactivates=true)]
  rendererClasses [&NiagaraMeshRendererProperties]
  renderers [{rendererClass=&NiagaraMeshRendererProperties, rendererIndex=0}]
 }
 SplineThicken_Loop001 {
  emitterUpdateScript [
   &EmitterState(Life Cycle Mode=::NewEnumerator1, Loop Behavior=::NewEnumerator1, Loop Duration Mode=::NewEnumerator1, UseLoopDelay=false)
   &SpawnBurst_Instantaneous(Use Spawn Probability=false, UseLoopCountLimit=false)
  ]
  particleSpawnScript [
   &InitializeParticle(
    Color Mode=::NewEnumerator0
    Lifetime=3
    Mesh Scale Mode=::NewEnumerator3
    Mesh Uniform Scale=$User.RingScale
    UsePositionOffset=false
    Write Material Random=true
   )
   &DynamicMaterialParameters(
    Index 0 Param 1=$User.RingScale
    Module.Index0Param0Write=true
    Module.Index0Param1Write=true
    Module.Index0Param2Write=true
    Module.Index0Param3Write=true
   )
   &InitialMeshOrientation(Use Rotation Vector=false) {enabled=false}
  ]
  particleUpdateScript [&ParticleState(Kill Particles When Lifetime Has Elapsed=false, Let Infinitely Lived Particles Die When Emitter Deactivates=true)]
  rendererClasses [&NiagaraMeshRendererProperties]
  renderers [{rendererClass=&NiagaraMeshRendererProperties, rendererIndex=0}]
 }
}
summary {
 systemName=NS_Sparks_RockSet_Pixels
 emitters [
  {emitterName=DotRing, simTarget=GPUComputeSim, rendererClasses [&NiagaraSpriteRendererProperties]}
  {emitterName=Boxes, simTarget=GPUComputeSim, rendererClasses [&NiagaraMeshRendererProperties]}
  {emitterName=SplineThicken_Loop, simTarget=GPUComputeSim, rendererClasses [&NiagaraMeshRendererProperties]}
  {emitterName=Boxes001, rendererClasses [&NiagaraMeshRendererProperties]}
  {emitterName=SplineThicken_Loop001, rendererClasses [&NiagaraMeshRendererProperties]}
 ]
 userVariables [
  User.BoxesScale {type=NiagaraFloat, defaultValue=1.5, description=""}
  User.RingScale {type=NiagaraFloat, defaultValue=1, description=""}
 ]
}
